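; Parent quest prop for the collectible FORTNITE letters, choosing its mesh from the letter index and scaling beam draw distance by game mode, platform and view-distance CVar.
; Entering the start overlap activates the beam and runs pop-up and looping spin timelines; leaving the exit overlap drops and shrinks the letter.
; On objective completion it gathers completed letter quests, spawns the FORTNITE word actor if at least two are recorded, and unbinds both overlaps.
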(event EventBeginPlay
  (|CallParentFunction) ; This node is disabled and will not be called.\nDrag off pins to build functionality.
)

(event EventTick
  (params (DeltaSeconds :type Float%20%28single-precision%29))
  (|CallParentFunction DeltaSeconds) ; This node is disabled and will not be called.\nDrag off pins to build functionality.
)

(event EventQuestInitializedAndValid
  (bind _staticmeshcomponent (Variables|Debug|GetStaticMeshComponent))
  ;[1 1 1] Initialize Letters BP
  (Variables|Default|SetLetterInitialLocation (Variables|/Script/Engine.SceneComponent|GetRelativeLocation _staticmeshcomponent))
  (Variables|Default|SetLetterInitialRotation (Variables|/Script/Engine.SceneComponent|GetRelativeRotation _staticmeshcomponent))
  (Collision|BindEventtoOnComponentBeginOverlap (Variables|Default|GetStartSpinOverlap) (delegate Custom|PawnCloseEnoughToTriggerBobbing))
  (Collision|BindEventtoOnComponentEndOverlap (Variables|Default|GetEndSpinExitOverlap) (delegate Custom|PawnFarEnoughToEndBobbing))
  (CallFunction|SetLetterAtBeginning)
  (CallFunction|SetMaxDrawDistance)
  (Transformation|SetRelativeScale3D (Variables|Default|GetSM_Fortnite_Letter) "0.001000,0.001000,0.001000"))

(event Custom|PawnCloseEnoughToTriggerBobbing
  (params (OverlappedComponent :type Primitive%20Component%20Object%20Reference) (OtherActor :type Actor%20Object%20Reference) (OtherComp :type Primitive%20Component%20Object%20Reference) (OtherBodyIndex :type Integer) (bFromSweep :type Boolean) (SweepResult :type Hit%20Result%20Structure%20%28by%20ref%29))
  ;[1 1 1] Enable spinning and bobbing letter
  (Components|Activation|Activate (Variables|Default|GetP_GodBeam) false)
  (bind (_direction _growtime) (Timeline|Pop%20Up%20and%20Down Play)
    (:Update
      (Transformation|SetRelativeScale3D (Variables|Default|GetSM_Fortnite_Letter) (Math|Vector|Lerp(Vector) "0.001000,0.001000,0.001000" "1.000000,1.000000,1.000000" _growtime)))
    (:Finished
      (switch Utilities|FlowControl|Switch|SwitchonETimelineDirection _direction
        (:Forward
          (Components|Timeline|SetLooping (Variables|Prop_QuestInteractable_Letters_Parent|GetBobAndSpin) true)
          (bind (_direction_2 _bob _spin) (Timeline|BobAndSpin PlayFromStart)
            (:Update
              (Transformation|SetWorldRotation (Variables|Default|GetSM_Fortnite_Letter) (Math|Rotator|Lerp(Rotator) "0, 0, 0" "0,720.000000, 0" _spin) false false))))))))

(event Custom|PawnFarEnoughToEndBobbing
  (params (OverlappedComponent :type Primitive%20Component%20Object%20Reference) (OtherActor :type Actor%20Object%20Reference) (OtherComp :type Primitive%20Component%20Object%20Reference) (OtherBodyIndex :type Integer))
  (bind _sm_fortnite_letter (Variables|Default|GetSM_Fortnite_Letter))
  ;[1 1 1] Disable spinning and bobbing Letter
  (Utilities|FlowControl|Sequence
    (:then_0
      _)
    (:then_1
      (Variables|Default|SetLetterCurrentLocation (Transformation|GetWorldLocation _sm_fortnite_letter))
      (Variables|Default|SetLetterCurrentRotation (Variables|/Script/Engine.SceneComponent|GetRelativeRotation _sm_fortnite_letter))
      (bind (_direction _droptime) (Timeline|Drop PlayFromStart)
        (:Update
          (Transformation|SetWorldLocationAndRotation (Variables|Default|GetSM_Fortnite_Letter) (Math|Vector|Lerp(Vector) (Variables|Default|GetLetterCurrentLocation) (Variables|Default|GetLetterInitialLocation) _droptime) (Math|Rotator|Lerp(Rotator) (Variables|Default|GetLetterCurrentRotation) (Variables|Default|GetLetterInitialRotation) _droptime) false false))))
    (:then_2
      (Components|Activation|Deactivate (Variables|Default|GetP_GodBeam))
      ;[1 1 1] Enable spinning and bobbing letter
      (Timeline|Pop%20Up%20and%20Down Reverse)))
  (Timeline|BobAndSpin Stop))

(event EventThisQuestObjectiveComplete
  (params (IneractingPawn :type Fort%20Pawn%20Object%20Reference))
  ;[1 1 1] Update Letters to match the next one you are looking for
  ;[1 0 0.05723] we're not using stat handle references anymore, i've removed them here, needs to be updated to use backend names or tags -nicholas.lacelle
  (for _array_element (Variables|Default|GetAll_Quests)
    (if (Quest|HasCompletedObjective :self (Quests|GetQuestManager (Utilities|Array|Get(acopy) (Fortnite|GetLocalFortPlayerControllers true)) "Invalid") :Definition _array_element :!class /Script/FortniteGame.FortQuestManager)
      (Utilities|Array|AddUnique (select (Variables|Default|GetDebugLetters) (Variables|Default|GetCompleted_Quests_Test) (Variables|Default|GetCompleted_Quests)) _array_element)))
  ;[1 1 1] Spawn FORTNITE word up to the number of letters you have collected so far
  (if (not (< (Utilities|Array|Length (Variables|Default|GetCompleted_Quests)) 2))
    (Game|SpawnActorfromClass :Class "/Game/Athena/Items/QuestInteractables/FortnightLetters/BA_FORTNITE_NonSequential.BA_FORTNITE_NonSequential_C" :SpawnTransform (Math|Transform|MakeTransform (Transformation|GetActorLocation) (Math|Conversions|RotationFromXVector (- (* (Transformation|GetActorLocation IneractingPawn) "1.000000,1.000000,0.000000") (* (Transformation|GetActorLocation) "1.000000,1.000000,0.000000")))) :CollisionHandlingOverride "AlwaysSpawn" :TransformScaleMethod "OverrideRootScale" :CurrentLetter (Variables|Default|GetCurrent%20Letter%20Int) :Completed_Letter_Quests (select (Variables|Default|GetDebugLetters) (Variables|Default|GetCompleted_Quests_Test) (Variables|Default|GetCompleted_Quests)) :Instigator IneractingPawn))
  ;[1 1 1] Unbind Bob And Spin Events
  (Collision|UnbindEventfromOnComponentBeginOverlap (Variables|Default|GetStartSpinOverlap) (EventDispatchers|CreateEvent :!event Custom|PawnCloseEnoughToTriggerBobbing))
  (Collision|UnbindEventfromOnComponentEndOverlap (Variables|Default|GetEndSpinExitOverlap) (EventDispatchers|CreateEvent :!event Custom|PawnFarEnoughToEndBobbing)))

(event Custom|Set%20Letter%20at%20Beginning
  (bind _sm_fortnite_letter (Variables|Default|GetSM_Fortnite_Letter))
  ;[1 1 1] Update Letters to match the next one you are looking for
  (switch int (Variables|Default|GetCurrent%20Letter%20Int)
    (:1
      (Components|StaticMesh|SetStaticMesh _sm_fortnite_letter "/Game/Athena/Items/QuestInteractables/FortnightLetters/FX/SM_Fortnite_O.SM_Fortnite_O"))
    (:2
      (Components|StaticMesh|SetStaticMesh _sm_fortnite_letter "/Game/Athena/Items/QuestInteractables/FortnightLetters/FX/SM_Fortnite_R.SM_Fortnite_R"))
    (:3
      (Components|StaticMesh|SetStaticMesh _sm_fortnite_letter "/Game/Athena/Items/QuestInteractables/FortnightLetters/FX/SM_Fortnite_T1.SM_Fortnite_T1"))
    (:4
      (Components|StaticMesh|SetStaticMesh _sm_fortnite_letter "/Game/Athena/Items/QuestInteractables/FortnightLetters/FX/SM_Fortnite_N.SM_Fortnite_N"))
    (:5
      (Components|StaticMesh|SetStaticMesh _sm_fortnite_letter "/Game/Athena/Items/QuestInteractables/FortnightLetters/FX/SM_Fortnite_I.SM_Fortnite_I"))
    (:6
      (Components|StaticMesh|SetStaticMesh _sm_fortnite_letter "/Game/Athena/Items/QuestInteractables/FortnightLetters/FX/SM_Fortnite_T2.SM_Fortnite_T2"))
    (:7
      (Components|StaticMesh|SetStaticMesh _sm_fortnite_letter "/Game/Athena/Items/QuestInteractables/FortnightLetters/FX/SM_Fortnite_E.SM_Fortnite_E"))))

(event EventQuestUpdatedFromAnywhere
  ;[1 1 1] Update Letters to match the next one you are looking for
  (|CallParentFunction))

(event Custom|SetMaxDrawDistance
  (bind (_asfort_game_state_athena _bsuccess) (Utilities|Casting|CastToFortGameStateAthena (Game|GetGameState)))
  ;[1 1 1] Set Max Draw Distance for Particle System
  (LOD|SetMaxDrawDistance (Variables|Default|GetP_GodBeam) (* (select _bsuccess (select (Fortnite|IsMobilePlatform) (Variables|Default|GetMax%20Draw%20Distance%20Bacchus) (Variables|Default|GetMax%20Draw%20Distance%20Athena)) (Variables|Default|GetMax%20Draw%20Distance%20StW)) (Utilities|Reciprocal (Fortnite|GetNumericCVar "r.ViewDistanceScale")))))

(fn ConstructionScript
  (|CallParentFunction))
